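; PCG graph template in Biome Core that defines the interface for assembly behavior graphs.
; It declares a TerrainSurface surface input and a RequiredAttributes attribute set output.
; The graph body is empty: no nodes or connections link the input to the output.
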(graph BiomeCore_AssemblyBehavior_Template
  (inputs (TerrainSurface :type Surface))
  (outputs (RequiredAttributes :type Param)))
