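; Default PCG Biome Core assembly behavior that resolves assembly points using per-point tag attributes and outputs the required attribute defaults.
; Seeded noise thins clutter by density tier, then tags randomize relative rotation and scale.
; A first hierarchy pass places points before terrain projection, where RTL keeps the projected rotation and STL restores its original.
; After Z offsets, a second hierarchy pass with STL opt-out precedes filtering by ActorIndex and CULL, then attribute cleanup.
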(graph BiomeCore_DefaultAssemblyBehavior
  (params (ClutterDensity :type Double :default 0.66) (ClutterSeed :type Int64 :default 101) (ClutterLow :type Double :default -0.25) (ClutterMed :type Double :default -0.1) (ClutterHigh :type Double :default 0.2) (ActorRepresentation :type Int :default 2) (NoLandscapeProjection :type Bool) (RequiredAttributes :type (Array String) :default ["IPR:false" "IPS:false" "STL:false" "RTL:false" "FixedScale:false" "RandScale5:false" "RandScale10:false" "RandScale15:false" "RandScale20:false" "bCastShadow:true" "WorldPositionOffsetDisableDistance:0.0" "bRenderInMainPass:true" "DetailMode:0" "CULL:false" "Clutter:false" "ClutterLow:false" "ClutterMed:false" "ClutterHigh:false" "bCastHiddenShadow:false" "InstanceStartCullDistance:0" "InstanceEndCullDistance:0" "bIncludeInHLOD:true" "bAffectDistanceFieldLighting:true" "bAffectDynamicIndirectLighting:true" "bVisibleInRayTracing:true" "bUseDefaultCollision:true" "OffsetZ:0" "RotZ:false" "RotY:false" "RotX:false" "bCanEverAffectNavigation:true" "bRasterizeAsFilledConvexVolume:false" "CastDynamicShadow:true" "EvaluateWorldPositionOffset:true" "IsSpatiallyLoaded:true" "ComponentTags:" "MeshOptionsPreset:"]) (TerrainGetterSettings :type (Object /Script/PCG.PCGSettingsInterface) :default /PCGBiomeCore/Setup/WorldRayHitSetting.WorldRayHitSetting))
  (inputs (TerrainSurface :type Surface))
  (outputs (RequiredAttributes :type Param))
  (bind DataNum_0 (Data%20Count))
  (bind Branch_0 (Branch))
  (bind MergePoints_11 (Merge%20Points))
  (bind ApplyHierarchy_0 (Apply%20Hierarchy :ApplyHierarchy OptOutByAttribute :ApplyHierarchyAttribute STL :ApplyParentRotationAttribute IPR :ApplyParentScaleAttribute IPS :bWarnOnPointsWithInvalidParent false :ParentKeyAttributes [ParentIndex TargetIndex] :PointKeyAttributes [ActorIndex TargetIndex]))
  (bind AttributeFilter_1 (Filter%20Attribute%20Elements :AttributeTypes (struct :Type Integer64 :Int32Value 1 :BoolValue true) :bUseConstantThreshold true :bUseSpatialQuery true :Operator GreaterOrEqual :TargetAttribute ActorIndex))
  (bind AttributeFilter_7 (Filter%20Attribute%20Elements :AttributeTypes (struct :Type Boolean :Int32Value 1 :BoolValue true) :bUseConstantThreshold true :bUseSpatialQuery true :Operator Equal :TargetAttribute CULL)) ; CULL == 1 Remove any points with "CULL" tag 
  (bind DeleteAttributes_8 (Delete%20Attributes :SelectedAttributes "ActorIndex,ParentIndex,TargetIndex,CULL,HierarchyDepth,RelativeTransform,IPR,IPS,STL"))
  ;[0 0 0] Clutter [TAG] Clutter [TAG] ClutterLow [TAG] ClutterMed [TAG] ClutterHigh
  (bind AttributeFilter_6 (Filter%20Attribute%20Elements :AttributeTypes (struct :Type Boolean :Int32Value 1 :BoolValue true) :bUseConstantThreshold true :bUseSpatialQuery true :Operator Equal :TargetAttribute Clutter)) ; Clutter == 1
  (bind AttributeFilter_9 (Filter%20Attribute%20Elements :AttributeTypes (struct :Type Boolean :Int32Value 1 :IntValue 1 :BoolValue true) :bUseConstantThreshold true :bUseSpatialQuery true :Operator Equal :TargetAttribute ClutterLow))
  (bind AttributeFilter_10 (Filter%20Attribute%20Elements :AttributeTypes (struct :Type Boolean :Int32Value 1 :IntValue 1 :BoolValue true) :bUseConstantThreshold true :bUseSpatialQuery true :Operator Equal :TargetAttribute ClutterMed))
  (bind AttributeFilter_11 (Filter%20Attribute%20Elements :AttributeTypes (struct :Type Boolean :Int32Value 1 :IntValue 1 :BoolValue true) :bUseConstantThreshold true :bUseSpatialQuery true :Operator Equal :TargetAttribute ClutterHigh))
  (bind AttributeNoise_2 (Attribute%20Noise :Seed-param ClutterSeed))
  (bind AddAttribute_0 (Add%20Attribute :AttributeTypes (struct :Type Float) :InputSource ClutterDensity :OutputTarget ClutterDensity :Attributes ClutterDensity))
  (bind AttributeFilter_8 (Filter%20Attribute%20Elements :bUseSpatialQuery true :bWarnOnDataMissingAttribute false :Operator LesserOrEqual :ThresholdAttribute ClutterDensity))
  (bind AttributeNoise_3 (Attribute%20Noise :Seed-param ClutterSeed))
  (bind AttributeMathsOp_1 (Attribute%20Maths%20Op :InputSource1 ClutterDensity :InputSource2 ClutterLow :Operation Multiply :OutputTarget None :InA ClutterDensity :InB ClutterLow))
  (bind AttributeMathsOp_2 (Attribute%20Maths%20Op :InputSource1 ClutterDensity :InputSource2 None))
  (bind DeleteAttributes_0 (Delete%20Attributes :bTokenizeOnWhiteSpace true :Operation KeepSelectedAttributes :SelectedAttributes "ClutterDensity"))
  (bind AddAttribute_1 (Add%20Attribute :AttributeTypes (struct :Type Float) :InputSource ClutterDensity :OutputTarget ClutterDensity))
  (bind AttributeFilter_16 (Filter%20Attribute%20Elements :bUseSpatialQuery true :bWarnOnDataMissingAttribute false :Operator LesserOrEqual :ThresholdAttribute ClutterDensity))
  (bind AttributeNoise_4 (Attribute%20Noise :Seed-param ClutterSeed))
  (bind AttributeMathsOp_7 (Attribute%20Maths%20Op :InputSource1 ClutterDensity :InputSource2 ClutterMed :Operation Multiply :OutputTarget None :InA ClutterDensity :InB ClutterMed))
  (bind AttributeMathsOp_10 (Attribute%20Maths%20Op :InputSource1 ClutterDensity :InputSource2 None))
  (bind DeleteAttributes_1 (Delete%20Attributes :bTokenizeOnWhiteSpace true :Operation KeepSelectedAttributes :SelectedAttributes "ClutterDensity"))
  (bind AddAttribute_2 (Add%20Attribute :AttributeTypes (struct :Type Float) :InputSource ClutterDensity :OutputTarget ClutterDensity))
  (bind AttributeFilter_17 (Filter%20Attribute%20Elements :bUseSpatialQuery true :bWarnOnDataMissingAttribute false :Operator LesserOrEqual :ThresholdAttribute ClutterDensity))
  (bind AttributeNoise_5 (Attribute%20Noise :Seed-param ClutterSeed))
  (bind AttributeMathsOp_12 (Attribute%20Maths%20Op :InputSource1 ClutterDensity :InputSource2 ClutterHigh :Operation Multiply :OutputTarget None :InA ClutterDensity :InB ClutterHigh))
  (bind AttributeMathsOp_15 (Attribute%20Maths%20Op :InputSource1 ClutterDensity :InputSource2 None))
  (bind DeleteAttributes_2 (Delete%20Attributes :bTokenizeOnWhiteSpace true :Operation KeepSelectedAttributes :SelectedAttributes "ClutterDensity"))
  (bind AddAttribute_3 (Add%20Attribute :AttributeTypes (struct :Type Float) :InputSource ClutterDensity :OutputTarget ClutterDensity))
  (bind AttributeFilter_20 (Filter%20Attribute%20Elements :bUseSpatialQuery true :bWarnOnDataMissingAttribute false :Operator LesserOrEqual :ThresholdAttribute ClutterDensity))
  (bind DeleteAttributes_5 (Delete%20Attributes :Operator Substring :SelectedAttributes "Clutter"))
  (bind MergePoints_9 (Merge%20Points))
  ;[0 0 0] Local Rotation TAGS
  ;;[0 0 0] [TAG] RotZ
  (bind AttributeFilter_0 (Filter%20Attribute%20Elements :AttributeTypes (struct :Type Boolean :Int32Value 1 :BoolValue true) :bUseConstantThreshold true :bUseSpatialQuery true :Operator Equal :TargetAttribute RotZ)) ; RotZ == 1
  (bind TransformPoints_0 (Transform%20Points :AttributeName RelativeTransform :bApplyToAttribute true :RotationMax (struct :Yaw 1.8e+02) :RotationMin (struct :Yaw -1.8e+02)))
  ;;[0 0 0] [TAG] RotX
  (bind AttributeFilter_2 (Filter%20Attribute%20Elements :AttributeTypes (struct :Type Boolean :Int32Value 1 :IntValue 1 :BoolValue true) :bUseConstantThreshold true :bUseSpatialQuery true :Operator Equal :TargetAttribute RotX)) ; RotX == 1
  (bind TransformPoints_1 (Transform%20Points :AttributeName RelativeTransform :bApplyToAttribute true :RotationMax (struct :Roll 1.8e+02) :RotationMin (struct :Roll -1.8e+02)))
  ;;[0 0 0] [TAG] RotY
  (bind AttributeFilter_3 (Filter%20Attribute%20Elements :AttributeTypes (struct :Type Boolean :Int32Value 1 :IntValue 1 :BoolValue true) :bUseConstantThreshold true :bUseSpatialQuery true :Operator Equal :TargetAttribute RotY)) ; RotY == 1
  (bind TransformPoints_2 (Transform%20Points :AttributeName RelativeTransform :bApplyToAttribute true :RotationMax (struct :Pitch 1.8e+02) :RotationMin (struct :Pitch -1.8e+02)))
  ;[0 0 0] Tag[FixedScale] and other scale based tags BP Random Scale Ranges
  (bind PointFilter_6 (Filter%20Attribute%20Elements :AttributeTypes (struct :Type Boolean :Int32Value 1 :IntValue 1 :BoolValue true) :bUseConstantThreshold true :bUseSpatialQuery true :bWarnOnDataMissingAttribute false :Operator Equal :TargetAttribute FixedScale)) ; [TAG]FixedScale
  (bind PointFilter_22 (Filter%20Attribute%20Elements :AttributeTypes (struct :Type Boolean :Int32Value 1 :IntValue 1 :BoolValue true) :bUseConstantThreshold true :bUseSpatialQuery true :bWarnOnDataMissingAttribute false :Operator Equal :TargetAttribute RandScale5)) ; [TAG]RandScale5
  (bind PointFilter_4 (Filter%20Attribute%20Elements :AttributeTypes (struct :Type Boolean :Int32Value 1 :IntValue 1 :BoolValue true) :bUseConstantThreshold true :bUseSpatialQuery true :bWarnOnDataMissingAttribute false :Operator Equal :TargetAttribute RandScale10)) ; [TAG]RandScale10
  (bind AttributeNoise_6 (Attribute%20Noise :NoiseMax 1.1 :NoiseMin 0.9 :OutputTarget RandScale)) ; Create RandScale 0.9 to 1.1
  (bind PointFilter_10 (Filter%20Attribute%20Elements :AttributeTypes (struct :Type Boolean :Int32Value 1 :IntValue 1 :BoolValue true) :bUseConstantThreshold true :bUseSpatialQuery true :bWarnOnDataMissingAttribute false :Operator Equal :TargetAttribute RandScale15)) ; [TAG]RandScale15
  (bind AttributeNoise_12 (Attribute%20Noise :NoiseMax 1.15 :NoiseMin 0.85 :OutputTarget RandScale)) ; Create RandScale 0.85 to 1.15
  (bind PointFilter_16 (Filter%20Attribute%20Elements :AttributeTypes (struct :Type Boolean :Int32Value 1 :IntValue 1 :BoolValue true) :bUseConstantThreshold true :bUseSpatialQuery true :bWarnOnDataMissingAttribute false :Operator Equal :TargetAttribute RandScale20)) ; [TAG]RandScale20
  (bind AttributeNoise_18 (Attribute%20Noise :NoiseMax 1.2 :NoiseMin 0.8 :OutputTarget RandScale)) ; Create RandScale 0.8 to 1.2
  (bind AttributeNoise_24 (Attribute%20Noise :NoiseMax 1.05 :NoiseMin 0.95 :OutputTarget RandScale)) ; Create RandScale 0.95 to 1.05
  (bind Gather_0 (Gather))
  (bind AttributeMathsOp_26 (Attribute%20Maths%20Op :InputSource1 RelativeTransform.Scale :InputSource2 RandScale :Operation Multiply :OutputTarget RelativeTransform.Scale)) ; Apply RandScale to RelativeTransform.Scale
  (bind DeleteAttributes_4 (Delete%20Attributes :Operator Substring :SelectedAttributes "FixedScale,RandScale"))
  (bind MergePoints_8 (Merge%20Points))
  ;[0 0 0] Apply Hierarchy here to apply transforms before project to Terrain
  (bind ApplyHierarchy_2 (Apply%20Hierarchy :ApplyParentRotationAttribute IPR :ApplyParentScaleAttribute IPS :bWarnOnPointsWithInvalidParent false :ParentKeyAttributes [ParentIndex TargetIndex] :PointKeyAttributes [ActorIndex TargetIndex]))
  ;[0 0 0] [STL] [RTL] Landscape Snapping and Align
  (bind CopyAttribute_93 (Copy%20Attributes :InputSource $Position :OutputTarget OldPos))
  (bind PointFilter_47 (Filter%20Attribute%20Elements :AttributeTypes (struct :Type Boolean :Int32Value 1 :BoolValue true) :bUseConstantThreshold true :bUseSpatialQuery true :Operator Equal :TargetAttribute RTL)) ; RTL == 1
  (bind PointFilter_49 (Filter%20Attribute%20Elements :AttributeTypes (struct :Type Boolean :Int32Value 1 :BoolValue true) :bUseConstantThreshold true :bUseSpatialQuery true :Operator Equal :TargetAttribute STL)) ; STL == 1
  (bind MergePoints_3 (Merge%20Points))
  (bind CopyAttribute_7 (Copy%20Attributes :InputSource $Transform.Rotation :OutputTarget RotBak))
  (bind Projection_41 (Projection)) ; RTL
  (bind AttributeMathsOp_81 (Attribute%20Maths%20Op :InputSource1 $Position :InputSource2 OldPos :Operation Subtract :OutputTarget STLDelta))
  (bind AttributeMathsOp_85 (Attribute%20Maths%20Op :InputSource1 STLDelta :InputSource2 RelativeTransform.Position :OutputTarget RelativeTransform.Position))
  (bind NormalToDensity_31 (Normal%20To%20Density :Offset -0.8 :Strength 14.0))
  (bind DensityFilter_29 (Density%20Filter))
  (bind AttributeFilter_12 (Filter%20Attribute%20Elements :AttributeTypes (struct :Type Boolean :BoolValue true) :bUseConstantThreshold true :bUseSpatialQuery true :Operator Equal :TargetAttribute RTL)) ; RTL == 1
  (bind CopyAttribute_10 (Copy%20Attributes :InputSource RotBak :OutputTarget $Transform.Rotation)) ; STL points copy back rotations
  (bind DeleteAttributes_7 (Delete%20Attributes :SelectedAttributes "ASM,RTL,OldPos,RotBak,STLDelta"))
  ;[9.5e-05 9.5e-05 9.5e-05] OffsetZ tag
  (bind AttributeMathsOp_0 (Attribute%20Maths%20Op :InputSource1 $Position.Z :InputSource2 OffsetZ :OutputTarget $Position.Z)) ; OffsetZ tag
  (reroute NamedRerouteDeclaration_0 TerrainSurface)
  (connect AttributeNoise_2 Out AddAttribute_0 In)
  (connect DeleteAttributes_0 Out AddAttribute_1 Attributes)
  (connect AttributeNoise_3 Out AddAttribute_1 In)
  (connect DeleteAttributes_1 Out AddAttribute_2 Attributes)
  (connect AttributeNoise_4 Out AddAttribute_2 In)
  (connect DeleteAttributes_2 Out AddAttribute_3 Attributes)
  (connect AttributeNoise_5 Out AddAttribute_3 In)
  (connect MergePoints_11 Out ApplyHierarchy_0 In)
  (connect MergePoints_8 Out ApplyHierarchy_2 In)
  (connect MergePoints_9 Out AttributeFilter_0 In)
  (connect ApplyHierarchy_0 Out AttributeFilter_1 In)
  (connect AttributeFilter_0 OutsideFilter AttributeFilter_2 In)
  (connect TransformPoints_0 Out AttributeFilter_2 In)
  (connect AttributeFilter_2 OutsideFilter AttributeFilter_3 In)
  (connect TransformPoints_1 Out AttributeFilter_3 In)
  (connect Branch_0 Output%20B AttributeFilter_6 In)
  (connect AttributeFilter_1 InsideFilter AttributeFilter_7 In)
  (connect AddAttribute_0 Out AttributeFilter_8 In)
  (connect AttributeFilter_6 InsideFilter AttributeFilter_9 In)
  (connect AttributeFilter_9 OutsideFilter AttributeFilter_10 In)
  (connect AttributeFilter_10 OutsideFilter AttributeFilter_11 In)
  (connect AttributeMathsOp_85 Out AttributeFilter_12 In)
  (connect AddAttribute_1 Out AttributeFilter_16 In)
  (connect AddAttribute_2 Out AttributeFilter_17 In)
  (connect AddAttribute_3 Out AttributeFilter_20 In)
  (connect DeleteAttributes_7 Out AttributeMathsOp_0 InA)
  (connect DeleteAttributes_7 Out AttributeMathsOp_0 InB)
  (connect AttributeMathsOp_1 Out AttributeMathsOp_2 InA)
  (connect AttributeMathsOp_1 Out AttributeMathsOp_2 InB)
  (connect AttributeMathsOp_7 Out AttributeMathsOp_10 InA)
  (connect AttributeMathsOp_7 Out AttributeMathsOp_10 InB)
  (connect AttributeMathsOp_12 Out AttributeMathsOp_15 InA)
  (connect AttributeMathsOp_12 Out AttributeMathsOp_15 InB)
  (connect Gather_0 Out AttributeMathsOp_26 InA)
  (connect Gather_0 Out AttributeMathsOp_26 InB)
  (connect Projection_41 Out AttributeMathsOp_81 InA)
  (connect Projection_41 Out AttributeMathsOp_81 InB)
  (connect AttributeMathsOp_81 Out AttributeMathsOp_85 InA)
  (connect AttributeMathsOp_81 Out AttributeMathsOp_85 InB)
  (connect AttributeFilter_11 OutsideFilter AttributeNoise_2 In)
  (connect AttributeFilter_9 InsideFilter AttributeNoise_3 In)
  (connect AttributeFilter_10 InsideFilter AttributeNoise_4 In)
  (connect AttributeFilter_11 InsideFilter AttributeNoise_5 In)
  (connect PointFilter_4 InsideFilter AttributeNoise_6 In)
  (connect PointFilter_10 InsideFilter AttributeNoise_12 In)
  (connect PointFilter_16 InsideFilter AttributeNoise_18 In)
  (connect PointFilter_22 InsideFilter AttributeNoise_24 In)
  (connect DataNum_0 Out Branch_0 bOutputToB)
  (connect DefaultInputNode In Branch_0 In)
  (connect MergePoints_3 Out CopyAttribute_7 Source)
  (connect MergePoints_3 Out CopyAttribute_7 Target)
  (connect AttributeFilter_12 OutsideFilter CopyAttribute_10 Source)
  (connect AttributeFilter_12 OutsideFilter CopyAttribute_10 Target)
  (connect ApplyHierarchy_2 Out CopyAttribute_93 Source)
  (connect ApplyHierarchy_2 Out CopyAttribute_93 Target)
  (connect DefaultInputNode In DataNum_0 In)
  (connect AttributeMathsOp_2 Out DeleteAttributes_0 In)
  (connect AttributeMathsOp_10 Out DeleteAttributes_1 In)
  (connect AttributeMathsOp_15 Out DeleteAttributes_2 In)
  (connect AttributeMathsOp_26 Out DeleteAttributes_4 In)
  (connect PointFilter_6 InsideFilter DeleteAttributes_4 In)
  (connect PointFilter_16 OutsideFilter DeleteAttributes_4 In)
  (connect AttributeFilter_6 OutsideFilter DeleteAttributes_5 In)
  (connect AttributeFilter_8 InsideFilter DeleteAttributes_5 In)
  (connect AttributeFilter_16 InsideFilter DeleteAttributes_5 In)
  (connect AttributeFilter_17 InsideFilter DeleteAttributes_5 In)
  (connect AttributeFilter_20 InsideFilter DeleteAttributes_5 In)
  (connect AttributeFilter_12 InsideFilter DeleteAttributes_7 In)
  (connect CopyAttribute_10 Out DeleteAttributes_7 In)
  (connect PointFilter_49 OutsideFilter DeleteAttributes_7 In)
  (connect AttributeFilter_7 OutsideFilter DeleteAttributes_8 In)
  (connect NormalToDensity_31 Out DensityFilter_29 In)
  (connect AttributeNoise_6 Out Gather_0 In)
  (connect AttributeNoise_12 Out Gather_0 In)
  (connect AttributeNoise_18 Out Gather_0 In)
  (connect AttributeNoise_24 Out Gather_0 In)
  (connect PointFilter_47 InsideFilter MergePoints_3 In)
  (connect PointFilter_49 InsideFilter MergePoints_3 In2)
  (connect DeleteAttributes_4 Out MergePoints_8 In)
  (connect DeleteAttributes_5 Out MergePoints_9 In)
  (connect AttributeMathsOp_0 Out MergePoints_11 In)
  (connect PointFilter_22 OutsideFilter PointFilter_4 In)
  (connect AttributeFilter_3 OutsideFilter PointFilter_6 In)
  (connect TransformPoints_2 Out PointFilter_6 In)
  (connect PointFilter_4 OutsideFilter PointFilter_10 In)
  (connect PointFilter_10 OutsideFilter PointFilter_16 In)
  (connect PointFilter_6 OutsideFilter PointFilter_22 In)
  (connect CopyAttribute_93 Out PointFilter_47 In)
  (connect PointFilter_47 OutsideFilter PointFilter_49 In)
  (connect CopyAttribute_7 Out Projection_41 In)
  (connect NamedRerouteDeclaration_0 Out Projection_41 Projection%20Target)
  (connect AttributeFilter_0 InsideFilter TransformPoints_0 In)
  (connect AttributeFilter_2 InsideFilter TransformPoints_1 In)
  (connect AttributeFilter_3 InsideFilter TransformPoints_2 In)
  (output Out DeleteAttributes_8)
  (output RequiredAttributes RequiredAttributes))
